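# LeavePartyToggle is a CommonButtonBase widget that toggles between leaving and staying with the party.
# A horizontal box holds the inherited input action icon, with a circular progress material, beside a TextToggle reading "Leave Party" when off and "Stay with Party" when on.
# Clicking flips the TextToggle, and IsLeavePartySelected returns its on state, so leave counts as selected while "Stay with Party" shows.
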WidgetBlueprint /FortniteCoreUI/Library/LeavePartyToggle
graphs [
 EventGraph {nodes [
   Event On Clicked()
   > SwitchToggle(self:Text Toggle Object Reference=Get TextToggle() -> (TextToggle:Text Toggle Object Reference))
  ]}
 IsLeavePartySelected {nodes [
   IsLeavePartySelected() {is=title K2Node_FunctionEntry}
   > Return Node!(execute, ReturnValue:bool=Is Toggle On!(self:Text Toggle Object Reference=Get TextToggle()) -> (Is Toggle On:bool))
  ]}
]
variables []
widgets {
 info {inheritedWidgetCount=1, parentClass=&CommonButtonBase, rootWidgetClass=&HorizontalBox, widgetCount=3}
 widgets [
  HorizontalBox_0 {widgetClassPath=&HorizontalBox, widget=HorizontalBox_0}
  InputActionWidget {
   widgetClassPath=&CommonActionWidget
   bInherited=true
   bIsVariable=true
   parent=HorizontalBox_0
   widget=InputActionWidget
   properties {
    designTimeKey=Gamepad_FaceButton_Top
    progressMaterialParam=Percentage
    progressMaterialBrush {
     bIsDynamicallyLoaded=false
     drawAs=Image
     imageType=NoImage
     mirroring=NoMirror
     resourceName=None
     resourceObject=/Game/UI/Materials/HollowCircularProgressBar
     tiling=NoTile
     imageSize=<36 36>
     margin=<ltrb 0 0 0 0>
     outlineSettings {
      bUseBrushTransparency=false
      roundingType=HalfHeightRadius
      width=0
      color {colorUseRule=UseColor_Specified, specifiedColor=<rgba 0 0 0 0>}
      cornerRadii=<0 0 0 0>
     }
     tintColor {colorUseRule=UseColor_Specified, specifiedColor=<rgba 1 1 0 1>}
     uVRegion {bIsValid=false, max=<0 0>, min=<0 0>}
    }
   }
   slot {class=HorizontalBoxSlot, properties {horizontalAlignment=HAlign_Right, verticalAlignment=VAlign_Center, padding=<ltrb 0 0 5 0>}}
  }
  TextToggle {
   widgetClassPath=/Game/UI/DiscoveryShared/Status/TextToggle
   bIsVariable=true
   parent=HorizontalBox_0
   widget=TextToggle
   properties {offText=Leave Party, onText=Stay with Party}
   slot {class=HorizontalBoxSlot, properties {size {sizeRule=Fill, value=1}}}
  }
 ]
}
